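# Control Rig for the FlameBride backpack cosmetic that adjusts the pack bone as the wearer's spine moves.
# Three chained spherical pose readers measure dyn_spine relative to dyn_pelvis, each with its own cone orientation and falloff.
# Each reader's output weights a Modify Transforms step that rotates the pack, with the first also adding a small Z offset.
# Bow and ribbon bones follow through the hierarchy, and one Euler-to-quaternion node is unconnected.
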ControlRigBlueprint /BRCosmetics/Accessories/FORT_Backpacks/Backpack_F_MED_FlameBride/Meshes/F_MED_FlameBride_Pack_ControlRig
graphs [
 RigVMFunctionLibrary {links [], nodes []}
 RigVMModel {nodes [
   RigUnit_BeginExecution = &RigUnit_BeginExecution()
   n1 = &RigUnit_SphericalPoseReader(
    ActiveRegionScaleFactors=(PositiveWidth=2.5,NegativeWidth=0.3,PositiveHeight=0.5,NegativeHeight=4)
    ActiveRegionSize=0.1
    Debug=(bDrawDebug=false,bDraw2D=False,bDrawLocalAxes=False,DebugScale=25,DebugSegments=20,DebugThickness=0.25)
    DriverAxis=(X=0,Y=-1,Z=0)
    DriverItem=(Type=Bone,Name="dyn_spine")
    FalloffRegionScaleFactors=(PositiveWidth=-3,NegativeWidth=0.6,PositiveHeight=0.7,NegativeHeight=-2)
    FalloffSize=0.2
    FlipHeightScaling=False
    FlipWidthScaling=true
    OptionalParentItem=(Type=Bone,Name="dyn_pelvis")
    RotationOffset=(X=-90,Y=8,Z=-20)
    ExecutePin=@RigUnit_BeginExecution.ExecutePin
   )
   > &RigUnit_ModifyTransforms(
    ItemToModify=((Item=(Type=Bone,Name="pack"),Transform=(Rotation=(X=-0.0999,Y=0.0133,Z=-0.2792,W=0.9549),Translation=(X=0,Y=0,Z=2),Scale3D=(X=1,Y=1,Z=1))))
    WeightMinimum=0
    ItemToModify.0.Transform.Rotation=&RigVMFunction_MathQuaternionFromEuler(Euler=(X=-20,Y=20,Z=-10)).Result
    Weight=@n1.OutputParam
   )
   > n2 = &RigUnit_SphericalPoseReader(
    ActiveRegionScaleFactors=(PositiveWidth=1.5,NegativeWidth=0.2,PositiveHeight=0.3,NegativeHeight=4)
    ActiveRegionSize=0.1
    Debug=(bDrawDebug=false,bDraw2D=False,bDrawLocalAxes=False,DebugScale=25,DebugSegments=20,DebugThickness=0.25)
    DriverAxis=(X=0,Y=-1,Z=0)
    DriverItem=(Type=Bone,Name="dyn_spine")
    FalloffRegionScaleFactors=(PositiveWidth=2,NegativeWidth=0.5,PositiveHeight=0.5,NegativeHeight=-2)
    FalloffSize=0.2
    FlipHeightScaling=False
    FlipWidthScaling=true
    OptionalParentItem=(Type=Bone,Name="dyn_pelvis")
    RotationOffset=(X=-120,Y=5,Z=-35)
   )
   > &RigUnit_ModifyTransforms(
    ItemToModify=((Item=(Type=Bone,Name="pack"),Transform=(Rotation=(X=-0.0999,Y=0.0133,Z=-0.2792,W=0.9549),Translation=(X=0,Y=0,Z=0),Scale3D=(X=1,Y=1,Z=1))))
    WeightMinimum=0
    ItemToModify.0.Transform.Rotation=&RigVMFunction_MathQuaternionFromEuler(Euler=(X=0,Y=13,Z=0)).Result
    Weight=@n2.OutputParam
   )
   > n3 = &RigUnit_SphericalPoseReader(
    ActiveRegionScaleFactors=(PositiveWidth=0.8,NegativeWidth=0.3,PositiveHeight=0.3,NegativeHeight=0.8)
    ActiveRegionSize=0.1
    Debug=(bDrawDebug=false,bDraw2D=False,bDrawLocalAxes=False,DebugScale=25,DebugSegments=20,DebugThickness=0.25)
    DriverAxis=(X=0,Y=-1,Z=0)
    DriverItem=(Type=Bone,Name="dyn_spine")
    FalloffRegionScaleFactors=(PositiveWidth=0.6,NegativeWidth=0.6,PositiveHeight=0.6,NegativeHeight=0.5)
    FalloffSize=0.2
    FlipHeightScaling=False
    FlipWidthScaling=true
    OptionalParentItem=(Type=Bone,Name="dyn_pelvis")
    RotationOffset=(X=-60,Y=-20,Z=-10)
   )
   > &RigUnit_ModifyTransforms(
    ItemToModify=((Item=(Type=Bone,Name="pack"),Transform=(Rotation=(X=-0.0999,Y=0.0133,Z=-0.2792,W=0.9549),Translation=(X=0,Y=0,Z=0),Scale3D=(X=1,Y=1,Z=1))))
    WeightMinimum=0
    ItemToModify.0.Transform.Rotation=&RigVMFunction_MathQuaternionFromEuler(Euler=(X=0,Y=10,Z=0)).Result
    Weight=@n3.OutputParam
   )
   &RigVMFunction_MathQuaternionFromEuler(Euler=(X=0,Y=15,Z=0))
  ]}
]
hierarchy [
 root {loc=<0 0 0>, rot=<0 0 0 1>}
 pack {parent=root, loc=<0 -0.7316 0>, rot=<0 0 0 1>}
 dyn_ribbon_MSTR {parent=pack, loc=<4.9085 -1.5562 -2.4476>, rot=<-0.0906 -0.6971 0.2124 -0.6787>}
 dyn_ribbon_base {parent=dyn_ribbon_MSTR, loc=<0 0 0>, rot=<0 0 0 1>}
 dyn_ribbon_loop {parent=dyn_ribbon_base, loc=<-0.419 0.0109 0.1779>, rot=<-0.0746 -0.9103 0.0333 0.4059>}
 dyn_ribbon {parent=dyn_ribbon_base, loc=<0.4444 0.0234 -0.1636>, rot=<0 0 0 1>}
 dyn_big_bow_MSTR {parent=pack, loc=<0.3421 -5.4495 1.4618>, rot=<-0.5 -0.5 0.5 -0.5>}
 dyn_big_bow_base {parent=dyn_big_bow_MSTR, loc=<0 0 0>, rot=<0 0 0 1>}
 dyn_big_bow_ribbon_00_l {parent=dyn_big_bow_base, loc=<0.3219 1.8644 -1.8014>, rot=<0.6809 0.2763 -0.1905 -0.6509>}
 dyn_big_bow_ribbon_01_l {parent=dyn_big_bow_ribbon_00_l, loc=<7.1084 0 0>, rot=<0 -0.2006 0 -0.9797>}
 dyn_big_bow_ribbon_02_l {parent=dyn_big_bow_ribbon_01_l, loc=<6.4744 0 0>, rot=<0 -0.0615 0 -0.9981>}
 dyn_big_bow_ribbon_00_r {parent=dyn_big_bow_base, loc=<0.31 -1.987 -1.5751>, rot=<-0.6861 0.2633 0.2029 -0.6472>}
 dyn_big_bow_ribbon_01_r {parent=dyn_big_bow_ribbon_00_r, loc=<7.1084 0 0>, rot=<-0.0107 -0.1658 -0.0635 -0.9841>}
 dyn_big_bow_ribbon_02_r {parent=dyn_big_bow_ribbon_01_r, loc=<6.2128 0 0>, rot=<0 -0.0569 0 -0.9984>}
 dyn_big_bow_loop_base_l {parent=dyn_big_bow_base, loc=<-0.0212 1.9014 0.0615>, rot=<0.5 0.5 -0.5 -0.5>}
 dyn_big_bow_loop_00_l {parent=dyn_big_bow_loop_base_l, loc=<0 0 0>, rot=<0 0 0 1>}
 dyn_big_bow_loop_01_l {parent=dyn_big_bow_loop_00_l, loc=<2.9599 -0.3808 -0.1207>, rot=<0 0 0 1>}
 dyn_big_bow_loop_02_l {parent=dyn_big_bow_loop_01_l, loc=<4.8765 0 0>, rot=<0 0 0.0956 -0.9954>}
 dyn_big_bow_loop_02_bot_l {parent=dyn_big_bow_loop_01_l, loc=<2.4053 0.0278 -4.1768>, rot=<0.0016 -0.2803 -0.0056 -0.9599>}
 dyn_big_bow_loop_02_up_l {parent=dyn_big_bow_loop_01_l, loc=<2.3356 0 4.4224>, rot=<-0.0016 0.2706 -0.0056 -0.9627>}
 dyn_big_bow_loop_base_r {parent=dyn_big_bow_base, loc=<0.0995 -2.1877 -0.0041>, rot=<-0.4659 0.5319 0.4659 -0.5319>}
 dyn_big_bow_loop_00_r {parent=dyn_big_bow_loop_base_r, loc=<0 0 0>, rot=<0 0 0 1>}
 dyn_big_bow_loop_01_r {parent=dyn_big_bow_loop_00_r, loc=<3.3874 0 0>, rot=<0 0 -0.0297 -0.9996>}
 dyn_big_bow_loop_02_r {parent=dyn_big_bow_loop_01_r, loc=<4.9672 0 0>, rot=<0 0 0 1>}
 dyn_big_bow_loop_02_up_r {parent=dyn_big_bow_loop_01_r, loc=<2.3316 -0.1386 4.4224>, rot=<0.008 0.2705 0.0286 -0.9623>}
 dyn_big_bow_loop_02_bot_r {parent=dyn_big_bow_loop_01_r, loc=<2.4012 -0.1427 -4.1768>, rot=<-0.0083 -0.2802 0.0285 -0.9595>}
 sword_hide {parent=pack, loc=<0 0 0>, rot=<0 0 0 1>}
 dyn_small_bow_MSTR {parent=sword_hide, loc=<8.3801 -2.8148 9.2484>, rot=<-0.2431 -0.6726 0.4728 -0.5147>}
 dyn_small_bow_base {parent=dyn_small_bow_MSTR, loc=<0 0 0>, rot=<0 0 0 1>}
 dyn_small_bow_loop_r {parent=dyn_small_bow_base, loc=<-0.1206 -0.519 0.0509>, rot=<-0.3999 0.4168 0.568 -0.5862>}
 dyn_small_bow_loop_l {parent=dyn_small_bow_base, loc=<0.0059 0.6039 0.1732>, rot=<0.2973 0.4016 -0.548 -0.6708>}
 dyn_small_bow_ribbon_00_r {parent=dyn_small_bow_base, loc=<-0.2547 -0.4053 -0.0934>, rot=<0.0286 0.569 0.6807 -0.4605>}
 dyn_small_bow_ribbon_01_r {parent=dyn_small_bow_ribbon_00_r, loc=<4.0011 0 0>, rot=<0 0 0 1>}
 dyn_small_bow_ribbon_00_l {parent=dyn_small_bow_base, loc=<-0.212 0.5098 0.0264>, rot=<0.0231 -0.4605 0.77 0.4411>}
 dyn_small_bow_ribbon_01_l {parent=dyn_small_bow_ribbon_00_l, loc=<4.0011 0 0>, rot=<0 0 0 1>}
 dyn_pelvis {parent=root, loc=<0 1.024 -5.704>, rot=<0 0 0 1>}
 dyn_spine {parent=dyn_pelvis, loc=<0 0 0>, rot=<0 0 0 1>}
]
