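# Animation Blueprint for the FlameBride backpack cosmetic, adding secondary motion to its bows and ribbons.
# Updates cache the parent's pelvis and spine_05 world transforms; the AnimGraph begins with world-space replace modifiers on dyn_pelvis and dyn_spine.
# A Control Rig then feeds chained per-bone AnimDynamics with tuned springs and angular limits.
# Skydive floating blends in a pack offset and switches one right ribbon to tighter limits.
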AnimBlueprint /BRCosmetics/Accessories/FORT_Backpacks/Backpack_F_MED_FlameBride/Meshes/F_MED_FlameBride_Pack_AnimBPP
target_skeleton=&/F_MED_FlameBride_Pack_Skeleton
graphs [
 AnimGraph {nodes [
   n1 = "Transform (Modify) Bone - Bone: dyn_pelvis"(ComponentPose=Local To Component(LocalPose=Input Pose()), Alpha=1, Translation="0,0,0", Rotation="0,0,0", Scale="1,1,1") {
    rotationMode=BMM_Replace
    rotationSpace=BCS_WorldSpace
    translationMode=BMM_Replace
    translationSpace=BCS_WorldSpace
    alphaBoolBlend {blendOption=Linear}
    boneToModify {boneName=dyn_pelvis}
   }
   n2 = "Transform (Modify) Bone - Bone: dyn_spine"(ComponentPose=@n1, Alpha=1, Translation="0,0,0", Rotation="0,0,0", Scale="1,1,1") {
    rotationMode=BMM_Replace
    rotationSpace=BCS_WorldSpace
    translationMode=BMM_Replace
    translationSpace=BCS_WorldSpace
    alphaBoolBlend {blendOption=Linear}
    boneToModify {boneName=dyn_spine}
   }
   n3 = "Transform (Modify) Bone - Bone: pack"!(ComponentPose=@n2, bAlphaBoolEnabled:bool=Get bIsSkydiveFloating(), Translation="0,-1.5,0", Rotation="0,0,-20", Scale) -> (Pose) {
    alphaInputType=Bool
    rotationMode=BMM_Additive
    rotationSpace=BCS_BoneSpace
    translationMode=BMM_Additive
    translationSpace=BCS_BoneSpace
    alphaBoolBlend {blendInTime=2, blendOption=Linear, blendOutTime=5}
    boneToModify {boneName=pack}
    rotation {roll=-20}
    translation {y=-1.5}
   }
   n4 = Control Rig(Source=Component To Local(ComponentPose=@n3)) {
    alphaBoolBlend {blendOption=Linear}
    controlRigAssetReference {blueprintRigClass=&/F_MED_FlameBride_Pack_ControlRig.F_MED_FlameBride_Pack_ControlRig_C}
   }
   n5 = Anim Dynamics - Bone: dyn_ribbon(ComponentPose=Local To Component(LocalPose=@n4), Alpha=1, ExternalForce="0,0,0") {
    angularDampingOverride=0.999
    angularSpringConstant=2000
    bAngularSpring=true
    bOverrideAngularDamping=true
    bUseGravityOverride=true
    alphaBoolBlend {blendOption=Linear}
    boundBone {boneName=dyn_ribbon}
    componentAppliedLinearAccClamp=<400 400 400>
    componentLinearAccScale=<0.6 0.6 0.6>
    componentLinearVelScale=<0.7 0.7 0.7>
    gravityOverride {z=-980}
    physicsBodyDefinitions [{
      boundBone {boneName=dyn_ribbon}
      boxExtents=<15 15 15>
      constraintSetup {angularLimitsMax=<3 15 25>, angularLimitsMin=<-3 -30 -2>, angularTarget=<1 0 0>}
      localJointOffset=<1.5 0 0>
     }]
   }
   n6 = Anim Dynamics - Bone: dyn_ribbon_loop(ComponentPose=@n5, Alpha=1, ExternalForce="0,0,0") {
    angularDampingOverride=0.999
    angularSpringConstant=5000
    bAngularSpring=true
    bOverrideAngularDamping=true
    bUseGravityOverride=true
    alphaBoolBlend {blendOption=Linear}
    boundBone {boneName=dyn_ribbon_loop}
    componentAppliedLinearAccClamp=<400 400 400>
    componentLinearAccScale=<0.6 0.6 0.6>
    componentLinearVelScale=<0.7 0.7 0.7>
    gravityOverride {z=-980}
    physicsBodyDefinitions [{
      boundBone {boneName=dyn_ribbon_loop}
      boxExtents=<15 15 15>
      constraintSetup {angularLimitsMax=<3 15 25>, angularLimitsMin=<-3 -30 -2>, angularTarget=<1 0 0>}
      localJointOffset=<1 0 0>
     }]
   }
   n7 = Anim Dynamics - Bone: dyn_big_bow_base(ComponentPose=@n6, Alpha=1, ExternalForce="0,0,0") {
    angularDampingOverride=0.999
    angularSpringConstant=10000
    bAngularSpring=true
    bOverrideAngularDamping=true
    bUseGravityOverride=true
    gravityScale=0.2
    alphaBoolBlend {blendOption=Linear}
    boundBone {boneName=dyn_big_bow_base}
    componentAppliedLinearAccClamp=<600 600 600>
    componentLinearAccScale=<0.5 0.5 0.5>
    componentLinearVelScale=<0.6 0.6 0.6>
    gravityOverride {z=-980}
    physicsBodyDefinitions [{
      boundBone {boneName=dyn_big_bow_base}
      boxExtents=<40 40 40>
      constraintSetup {angularLimitsMax=<0 0 5>, angularLimitsMin=<0 0 -5>, angularTarget=<1 0 0>}
      localJointOffset=<2 0 0>
     }]
   }
   n8 = Anim Dynamics - Bone: dyn_big_bow_loop_00_l(ComponentPose=@n7, Alpha=1, ExternalForce="0,0,0") {
    angularDampingOverride=1
    angularSpringConstant=2000
    bOverrideAngularDamping=true
    bUseGravityOverride=true
    bUsePlanarLimit=false
    gravityScale=0.1
    linearSpringConstant=700
    alphaBoolBlend {blendOption=Linear}
    boundBone {boneName=dyn_big_bow_loop_00_l}
    componentAppliedLinearAccClamp=<600 600 600>
    componentLinearAccScale=<0.4 0.4 0.4>
    componentLinearVelScale=<0.5 0.5 0.5>
    gravityOverride {z=-500}
    physicsBodyDefinitions [{
      sphereCollisionRadius=1
      boundBone {boneName=dyn_big_bow_loop_00_l}
      boxExtents=<12 12 12>
      constraintSetup {angularLimitsMax=<0 5 2>, angularLimitsMin=<0 -12 -12>, angularTarget=<1 0 0>}
      localJointOffset=<1 0 -1>
     }]
   }
   n9 = Anim Dynamics - Bone: dyn_big_bow_loop_01_l(ComponentPose=@n8, Alpha=1, ExternalForce="0,0,0") {
    angularDampingOverride=1
    angularSpringConstant=2000
    bOverrideAngularDamping=true
    bUseGravityOverride=true
    bUsePlanarLimit=false
    gravityScale=0.1
    linearSpringConstant=700
    alphaBoolBlend {blendOption=Linear}
    boundBone {boneName=dyn_big_bow_loop_01_l}
    componentAppliedLinearAccClamp=<500 500 500>
    componentLinearAccScale=<0.4 0.4 0.4>
    componentLinearVelScale=<0.5 0.5 0.5>
    gravityOverride {z=-500}
    physicsBodyDefinitions [{
      sphereCollisionRadius=1
      boundBone {boneName=dyn_big_bow_loop_01_l}
      boxExtents=<12 11 12>
      constraintSetup {angularLimitsMax=<0 0 0>, angularLimitsMin=<0 -7 -12>, angularTarget=<1 0 0>}
      localJointOffset=<1 0 -1>
     }]
   }
   n10 = Anim Dynamics - Bone: dyn_big_bow_loop_02_up_l(ComponentPose=@n9, Alpha=1, ExternalForce="0,0,0") {
    angularDampingOverride=0.9
    angularSpringConstant=2000
    bOverrideAngularDamping=true
    bUseGravityOverride=true
    bUsePlanarLimit=false
    gravityScale=0.1
    linearSpringConstant=700
    alphaBoolBlend {blendOption=Linear}
    boundBone {boneName=dyn_big_bow_loop_02_up_l}
    componentAppliedLinearAccClamp=<500 500 500>
    componentLinearAccScale=<0.4 0.4 0.4>
    componentLinearVelScale=<0.6 0.6 0.6>
    gravityOverride {z=-500}
    physicsBodyDefinitions [{
      sphereCollisionRadius=1
      boundBone {boneName=dyn_big_bow_loop_02_up_l}
      boxExtents=<10 10 10>
      constraintSetup {angularLimitsMax=<0 0 5>, angularLimitsMin=<0 -5 -20>, angularTarget=<1 0 0>}
      localJointOffset=<1 0 -1>
     }]
   }
   n11 = Anim Dynamics - Bone: dyn_big_bow_loop_02_l(ComponentPose=@n10, Alpha=1, ExternalForce="0,0,0") {
    angularDampingOverride=0.9
    angularSpringConstant=2000
    bOverrideAngularDamping=true
    bUseGravityOverride=true
    bUsePlanarLimit=false
    gravityScale=0.1
    linearSpringConstant=700
    alphaBoolBlend {blendOption=Linear}
    boundBone {boneName=dyn_big_bow_loop_02_l}
    componentAppliedLinearAccClamp=<700 700 700>
    componentLinearAccScale=<0.4 0.4 0.4>
    componentLinearVelScale=<0.6 0.6 0.6>
    gravityOverride {z=-500}
    physicsBodyDefinitions [{
      sphereCollisionRadius=1
      boundBone {boneName=dyn_big_bow_loop_02_l}
      boxExtents=<10 10 10>
      constraintSetup {angularLimitsMax=<0 0 2>, angularLimitsMin=<0 -5 -20>, angularTarget=<1 0 0>}
      localJointOffset=<2 0 -1>
     }]
   }
   n12 = Anim Dynamics - Bone: dyn_big_bow_loop_02_bot_l(ComponentPose=@n11, Alpha=1, ExternalForce="0,0,0") {
    angularDampingOverride=0.9
    angularSpringConstant=2000
    bOverrideAngularDamping=true
    bUseGravityOverride=true
    bUsePlanarLimit=false
    gravityScale=0.1
    linearSpringConstant=700
    alphaBoolBlend {blendOption=Linear}
    boundBone {boneName=dyn_big_bow_loop_02_bot_l}
    componentAppliedLinearAccClamp=<700 700 700>
    componentLinearAccScale=<0.4 0.4 0.4>
    componentLinearVelScale=<0.6 0.6 0.6>
    gravityOverride {z=-500}
    physicsBodyDefinitions [{
      sphereCollisionRadius=1
      boundBone {boneName=dyn_big_bow_loop_02_bot_l}
      boxExtents=<10 10 10>
      constraintSetup {angularLimitsMax=<0 0 5>, angularLimitsMin=<0 -5 -15>, angularTarget=<1 0 0>}
      localJointOffset=<1 0 -1>
     }]
   }
   n13 = Anim Dynamics - Bone: dyn_big_bow_loop_00_r(ComponentPose=@n12, Alpha=1, ExternalForce="0,0,0") {
    angularDampingOverride=1
    angularSpringConstant=4000
    bOverrideAngularDamping=true
    bUseGravityOverride=true
    bUsePlanarLimit=false
    gravityScale=0.1
    linearSpringConstant=700
    alphaBoolBlend {blendOption=Linear}
    boundBone {boneName=dyn_big_bow_loop_00_r}
    componentAppliedLinearAccClamp=<600 600 600>
    componentLinearAccScale=<0.4 0.4 0.4>
    componentLinearVelScale=<0.5 0.5 0.5>
    gravityOverride {z=-500}
    physicsBodyDefinitions [{
      boundBone {boneName=dyn_big_bow_loop_00_r}
      boxExtents=<10 8 9>
      constraintSetup {angularLimitsMax=<0 5 14>, angularLimitsMin=<0 -12 -2>, angularTarget=<1 0 0>}
      localJointOffset=<1.1 0 -1>
     }]
   }
   n14 = Anim Dynamics - Bone: dyn_big_bow_loop_01_r(ComponentPose=@n13, Alpha=1, ExternalForce="0,0,0") {
    angularDampingOverride=1
    angularSpringConstant=4000
    bOverrideAngularDamping=true
    bUseGravityOverride=true
    bUsePlanarLimit=false
    gravityScale=0.1
    linearSpringConstant=700
    alphaBoolBlend {blendOption=Linear}
    boundBone {boneName=dyn_big_bow_loop_01_r}
    componentAppliedLinearAccClamp=<500 500 500>
    componentLinearAccScale=<0.4 0.4 0.4>
    componentLinearVelScale=<0.5 0.5 0.5>
    gravityOverride {z=-500}
    physicsBodyDefinitions [{
      boundBone {boneName=dyn_big_bow_loop_01_r}
      boxExtents=<10 10 10>
      constraintSetup {angularLimitsMax=<0 0 15>, angularLimitsMin=<0 -7 -2>, angularTarget=<1 0 0>}
      localJointOffset=<1.2 0 -1>
     }]
   }
   n15 = Anim Dynamics - Bone: dyn_big_bow_loop_02_up_r(ComponentPose=@n14, Alpha=1, ExternalForce="0,0,0") {
    angularDampingOverride=0.9
    angularSpringConstant=4000
    bOverrideAngularDamping=true
    bUseGravityOverride=true
    bUsePlanarLimit=false
    gravityScale=0.1
    linearSpringConstant=700
    alphaBoolBlend {blendOption=Linear}
    boundBone {boneName=dyn_big_bow_loop_02_up_r}
    componentAppliedLinearAccClamp=<500 500 500>
    componentLinearAccScale=<0.4 0.4 0.4>
    componentLinearVelScale=<0.6 0.6 0.6>
    gravityOverride {z=-500}
    physicsBodyDefinitions [{
      boundBone {boneName=dyn_big_bow_loop_02_up_r}
      boxExtents=<10 10 10>
      constraintSetup {angularLimitsMax=<0 0 20>, angularLimitsMin=<0 -5 -5>, angularTarget=<1 0 0>}
      localJointOffset=<1 0 -1>
     }]
   }
   n16 = Anim Dynamics - Bone: dyn_big_bow_loop_02_r(ComponentPose=@n15, Alpha=1, ExternalForce="0,0,0") {
    angularDampingOverride=0.9
    angularSpringConstant=4000
    bOverrideAngularDamping=true
    bUseGravityOverride=true
    bUsePlanarLimit=false
    gravityScale=0.1
    linearSpringConstant=700
    alphaBoolBlend {blendOption=Linear}
    boundBone {boneName=dyn_big_bow_loop_02_r}
    componentAppliedLinearAccClamp=<700 700 700>
    componentLinearAccScale=<0.4 0.4 0.4>
    componentLinearVelScale=<0.6 0.6 0.6>
    gravityOverride {z=-500}
    physicsBodyDefinitions [{
      boundBone {boneName=dyn_big_bow_loop_02_r}
      boxExtents=<10 10 10>
      constraintSetup {angularLimitsMax=<0 0 20>, angularLimitsMin=<0 -5 -2>, angularTarget=<1 0 0>}
      localJointOffset=<2 0 -1>
     }]
   }
   n17 = Anim Dynamics - Bone: dyn_big_bow_loop_02_bot_r(ComponentPose=@n16, Alpha=1, ExternalForce="0,0,0") {
    angularDampingOverride=0.9
    angularSpringConstant=4000
    bOverrideAngularDamping=true
    bUseGravityOverride=true
    bUsePlanarLimit=false
    gravityScale=0.1
    linearSpringConstant=700
    alphaBoolBlend {blendOption=Linear}
    boundBone {boneName=dyn_big_bow_loop_02_bot_r}
    componentAppliedLinearAccClamp=<700 700 700>
    componentLinearAccScale=<0.4 0.4 0.4>
    componentLinearVelScale=<0.6 0.6 0.6>
    gravityOverride {z=-500}
    physicsBodyDefinitions [{
      boundBone {boneName=dyn_big_bow_loop_02_bot_r}
      boxExtents=<10 10 10>
      constraintSetup {angularLimitsMax=<0 0 15>, angularLimitsMin=<0 -5 -5>, angularTarget=<1 0 0>}
      localJointOffset=<1 0 -1>
     }]
   }
   n18 = Anim Dynamics - Bone: dyn_big_bow_ribbon_00_l(ComponentPose=@n17, Alpha=1, ExternalForce="0,0,0") {
    angularDampingOverride=0.9999
    angularSpringConstant=1000
    bAngularSpring=true
    bOverrideAngularDamping=true
    bUseGravityOverride=true
    alphaBoolBlend {blendOption=Linear}
    boundBone {boneName=dyn_big_bow_ribbon_00_l}
    componentAppliedLinearAccClamp=<800 800 800>
    componentLinearAccScale=<0.6 0.6 0.6>
    componentLinearVelScale=<0.7 0.7 0.7>
    gravityOverride {z=-980}
    physicsBodyDefinitions [{
      boundBone {boneName=dyn_big_bow_ribbon_00_l}
      boxExtents=<14 14 14>
      constraintSetup {angularLimitsMax=<3 35 5>, angularLimitsMin=<-3 -35 -2>, angularTarget=<1 0 0>}
      localJointOffset=<2 0 -1.6>
     }]
   }
   n19 = Get bIsSkydiveFloating()
   n20 = Anim Dynamics - Bone: dyn_big_bow_ribbon_00_r!(ComponentPose=@n18, bAlphaBoolEnabled:bool=NOT Boolean(A=@n19), ExternalForce="0,0,0") -> (Pose) {
    alphaInputType=Bool
    angularDampingOverride=0.9999
    angularSpringConstant=1000
    bOverrideAngularDamping=true
    bUseGravityOverride=true
    alphaBoolBlend {blendOption=Linear}
    boundBone {boneName=dyn_big_bow_ribbon_00_r}
    componentAppliedLinearAccClamp=<800 800 800>
    componentLinearAccScale=<0.6 0.6 0.6>
    componentLinearVelScale=<0.7 0.7 0.7>
    gravityOverride {z=-980}
    physicsBodyDefinitions [{
      boundBone {boneName=dyn_big_bow_ribbon_00_r}
      boxExtents=<14 14 14>
      constraintSetup {angularLimitsMax=<3 35 5>, angularLimitsMin=<-3 -35 -5>, angularTarget=<1 0 0>}
      localJointOffset=<2 0 -1.5>
     }]
   }
   n21 = Anim Dynamics - Bone: dyn_big_bow_ribbon_00_r(ComponentPose=@n20, bAlphaBoolEnabled=@n19, ExternalForce="0,0,0") {
    alphaInputType=Bool
    angularDampingOverride=0.9999
    angularSpringConstant=1000
    bOverrideAngularDamping=true
    bUseGravityOverride=true
    alphaBoolBlend {blendOption=Linear}
    boundBone {boneName=dyn_big_bow_ribbon_00_r}
    componentAppliedLinearAccClamp=<800 800 800>
    componentLinearAccScale=<0.6 0.6 0.6>
    componentLinearVelScale=<0.7 0.7 0.7>
    gravityOverride {z=-980}
    physicsBodyDefinitions [{
      boundBone {boneName=dyn_big_bow_ribbon_00_r}
      boxExtents=<14 14 14>
      constraintSetup {angularLimitsMax=<3 35 5>, angularLimitsMin=<-3 -23 -5>, angularTarget=<1 0 0>}
      localJointOffset=<2 0 -1.5>
     }]
   }
   n22 = Anim Dynamics - Bone: dyn_big_bow_ribbon_01_l(ComponentPose=@n21, Alpha=1, ExternalForce="0,0,0") {
    angularDampingOverride=1
    angularSpringConstant=1000
    bAngularSpring=true
    bOverrideAngularDamping=true
    bUseGravityOverride=true
    alphaBoolBlend {blendOption=Linear}
    boundBone {boneName=dyn_big_bow_ribbon_01_l}
    componentAppliedLinearAccClamp=<800 800 800>
    componentLinearAccScale=<0.6 0.6 0.6>
    componentLinearVelScale=<0.7 0.7 0.7>
    gravityOverride {z=-980}
    physicsBodyDefinitions [{
      boundBone {boneName=dyn_big_bow_ribbon_01_l}
      boxExtents=<15 15 15>
      constraintSetup {angularLimitsMax=<3 30 5>, angularLimitsMin=<-3 -20 -8>, angularTarget=<1 0 0>}
      localJointOffset=<2 0 -0.6>
     }]
   }
   n23 = Anim Dynamics - Bone: dyn_big_bow_ribbon_01_r(ComponentPose=@n22, Alpha=1, ExternalForce="0,0,0") {
    angularDampingOverride=1
    angularSpringConstant=1000
    bAngularSpring=true
    bOverrideAngularDamping=true
    bUseGravityOverride=true
    alphaBoolBlend {blendOption=Linear}
    boundBone {boneName=dyn_big_bow_ribbon_01_r}
    componentAppliedLinearAccClamp=<800 800 800>
    componentLinearAccScale=<0.6 0.6 0.6>
    componentLinearVelScale=<0.7 0.7 0.7>
    gravityOverride {z=-980}
    physicsBodyDefinitions [{
      boundBone {boneName=dyn_big_bow_ribbon_01_r}
      boxExtents=<15 15 15>
      constraintSetup {angularLimitsMax=<3 30 8>, angularLimitsMin=<-3 -30 -5>, angularTarget=<1 0 0>}
      localJointOffset=<2 0 -0.6>
     }]
   }
   n24 = Anim Dynamics - Bone: dyn_big_bow_ribbon_02_l(ComponentPose=@n23, Alpha=1, ExternalForce="0,0,0") {
    angularDampingOverride=1
    angularSpringConstant=1000
    bAngularSpring=true
    bOverrideAngularDamping=true
    bUseGravityOverride=true
    alphaBoolBlend {blendOption=Linear}
    boundBone {boneName=dyn_big_bow_ribbon_02_l}
    componentAppliedLinearAccClamp=<800 800 800>
    componentLinearAccScale=<0.6 0.6 0.6>
    componentLinearVelScale=<0.7 0.7 0.7>
    gravityOverride {z=-980}
    physicsBodyDefinitions [{
      boundBone {boneName=dyn_big_bow_ribbon_02_l}
      boxExtents=<15 15 15>
      constraintSetup {angularLimitsMax=<3 30 8>, angularLimitsMin=<-3 -30 -30>, angularTarget=<1 0 0>}
      localJointOffset=<2 0.1 -0.2>
     }]
   }
   n25 = Anim Dynamics - Bone: dyn_big_bow_ribbon_02_r(ComponentPose=@n24, Alpha=1, ExternalForce="0,0,0") {
    angularDampingOverride=1
    angularSpringConstant=1000
    bAngularSpring=true
    bOverrideAngularDamping=true
    bUseGravityOverride=true
    alphaBoolBlend {blendOption=Linear}
    boundBone {boneName=dyn_big_bow_ribbon_02_r}
    componentAppliedLinearAccClamp=<800 800 800>
    componentLinearAccScale=<0.6 0.6 0.6>
    componentLinearVelScale=<0.7 0.7 0.7>
    gravityOverride {z=-980}
    physicsBodyDefinitions [{
      boundBone {boneName=dyn_big_bow_ribbon_02_r}
      boxExtents=<15 15 15>
      constraintSetup {angularLimitsMax=<3 30 30>, angularLimitsMin=<-3 -30 -8>, angularTarget=<1 0 0>}
      localJointOffset=<2 -0.1 -0.2>
     }]
   }
   n26 = Anim Dynamics - Bone: dyn_small_bow_loop_l(ComponentPose=@n25, Alpha=1, ExternalForce="0,0,0") {
    angularDampingOverride=0.9999
    angularSpringConstant=2000
    bAngularSpring=true
    bOverrideAngularDamping=true
    bUseGravityOverride=true
    alphaBoolBlend {blendOption=Linear}
    boundBone {boneName=dyn_small_bow_loop_l}
    componentAppliedLinearAccClamp=<400 400 400>
    componentLinearAccScale=<0.6 0.6 0.6>
    componentLinearVelScale=<0.7 0.7 0.7>
    gravityOverride {z=-600}
    physicsBodyDefinitions [{
      boundBone {boneName=dyn_small_bow_loop_l}
      boxExtents=<10 10 10>
      constraintSetup {angularLimitsMax=<3 25 25>, angularLimitsMin=<-3 -30 -2>, angularTarget=<1 0 0>}
      localJointOffset=<0.5 0 0>
     }]
   }
   n27 = Anim Dynamics - Bone: dyn_small_bow_loop_r(ComponentPose=@n26, Alpha=1, ExternalForce="0,0,0") {
    angularDampingOverride=0.9999
    angularSpringConstant=2000
    bAngularSpring=true
    bOverrideAngularDamping=true
    bUseGravityOverride=true
    alphaBoolBlend {blendOption=Linear}
    boundBone {boneName=dyn_small_bow_loop_r}
    componentAppliedLinearAccClamp=<400 400 400>
    componentLinearAccScale=<0.6 0.6 0.6>
    componentLinearVelScale=<0.7 0.7 0.7>
    gravityOverride {z=-600}
    physicsBodyDefinitions [{
      boundBone {boneName=dyn_small_bow_loop_r}
      boxExtents=<10 10 10>
      constraintSetup {angularLimitsMax=<3 25 25>, angularLimitsMin=<-3 -20 -30>, angularTarget=<1 0 0>}
      localJointOffset=<0.5 0 0>
     }]
   }
   n28 = Anim Dynamics - Bone: dyn_small_bow_ribbon_00_l(ComponentPose=@n27, Alpha=1, ExternalForce="0,0,0") {
    angularDampingOverride=0.9999
    angularSpringConstant=1000
    bAngularSpring=true
    bOverrideAngularDamping=true
    bUseGravityOverride=true
    alphaBoolBlend {blendOption=Linear}
    boundBone {boneName=dyn_small_bow_ribbon_00_l}
    componentAppliedLinearAccClamp=<400 400 400>
    componentLinearAccScale=<0.6 0.6 0.6>
    componentLinearVelScale=<0.7 0.7 0.7>
    gravityOverride {z=-600}
    physicsBodyDefinitions [{
      boundBone {boneName=dyn_small_bow_ribbon_00_l}
      boxExtents=<18 18 18>
      constraintSetup {angularLimitsMax=<3 20 20>, angularLimitsMin=<-3 -40 -10>, angularTarget=<1 0 0>}
      localJointOffset=<-1.5 -0.6 -3>
     }]
   }
   n29 = Anim Dynamics - Bone: dyn_small_bow_ribbon_00_r(ComponentPose=@n28, Alpha=1, ExternalForce="0,0,0") {
    angularDampingOverride=0.9999
    angularSpringConstant=1000
    bAngularSpring=true
    bOverrideAngularDamping=true
    bUseGravityOverride=true
    alphaBoolBlend {blendOption=Linear}
    boundBone {boneName=dyn_small_bow_ribbon_00_r}
    componentAppliedLinearAccClamp=<400 400 400>
    componentLinearAccScale=<0.6 0.6 0.6>
    componentLinearVelScale=<0.7 0.7 0.7>
    gravityOverride {z=-600}
    physicsBodyDefinitions [{
      boundBone {boneName=dyn_small_bow_ribbon_00_r}
      boxExtents=<14 14 14>
      constraintSetup {angularLimitsMax=<3 5 10>, angularLimitsMin=<-3 -40 -40>, angularTarget=<1 0 0>}
      localJointOffset=<-1.5 0.5 0>
     }]
   }
   n30 = Anim Dynamics - Bone: dyn_small_bow_ribbon_01_l(ComponentPose=@n29, Alpha=1, ExternalForce="0,0,0") {
    angularDampingOverride=0.9999
    angularSpringConstant=1000
    bAngularSpring=true
    bOverrideAngularDamping=true
    bUseGravityOverride=true
    alphaBoolBlend {blendOption=Linear}
    boundBone {boneName=dyn_small_bow_ribbon_01_l}
    componentAppliedLinearAccClamp=<400 400 400>
    componentLinearAccScale=<0.6 0.6 0.6>
    componentLinearVelScale=<0.7 0.7 0.7>
    gravityOverride {z=-600}
    physicsBodyDefinitions [{
      boundBone {boneName=dyn_small_bow_ribbon_01_l}
      boxExtents=<18 18 18>
      constraintSetup {angularLimitsMax=<3 30 20>, angularLimitsMin=<-3 -20 -10>, angularTarget=<1 0 0>}
      localJointOffset=<-1.5 -2 -1>
     }]
   }
   n31 = Anim Dynamics - Bone: dyn_small_bow_ribbon_01_r(ComponentPose=@n30, Alpha=1, ExternalForce="0,0,0") {
    angularDampingOverride=0.9999
    angularSpringConstant=3000
    bAngularSpring=true
    bOverrideAngularDamping=true
    bUseGravityOverride=true
    alphaBoolBlend {blendOption=Linear}
    boundBone {boneName=dyn_small_bow_ribbon_01_r}
    componentAppliedLinearAccClamp=<400 400 400>
    componentLinearAccScale=<0.6 0.6 0.6>
    componentLinearVelScale=<0.7 0.7 0.7>
    gravityOverride {z=-600}
    physicsBodyDefinitions [{
      boundBone {boneName=dyn_small_bow_ribbon_01_r}
      boxExtents=<14 14 14>
      constraintSetup {angularLimitsMax=<3 30 5>, angularLimitsMin=<-3 -20 -20>, angularTarget=<1 0 0>}
      localJointOffset=<-1.5 0.8 0>
     }]
   }
   Output Pose(Result=Component To Local(ComponentPose=@n31))
  ]}
 EventGraph {nodes [
   Event BlueprintUpdateAnimation()
   n1 = GetPartSkeletalMeshComponent()
   > Is Valid(InputObject=@n1)
   n2 = GetAttachParent!(self:Scene Component Object Reference=@n1) -> (ReturnValue:Scene Component Object Reference)
   > Is Valid(InputObject=@n2)
   n3 = BreakTransform(InTransform=GetSocketTransform(self=@n2, InSocketName=spine_05, TransformSpace=RTS_World))
   > Set Spine translates(Spine translates:Vector=@n3) -> (Output_Get:Vector)
   > Set Spine rotates(Spine rotates:Rotator=@n3.Rotation) -> (Output_Get:Rotator)
   n4 = BreakTransform(InTransform=GetSocketTransform(self=@n2, InSocketName=pelvis, TransformSpace=RTS_World))
   > Set Pelvis translates(Pelvis translates:Vector=@n4) -> (Output_Get:Vector)
   > Set Pelvis rotates(Pelvis rotates:Rotator=@n4.Rotation) -> (Output_Get:Rotator)
   TryGetPawnOwner()
  ]}
]
variables [
 Pelvis rotates {type=Rotator}
 Pelvis translates {type=Vector}
 Spine rotates {type=Rotator}
 Spine translates {type=Vector}
]
